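# Post-process style face accessory AnimBP for the ChOneJonesy helmet, adding secondary motion to its can and pot bones.
# It copies the attached parent's pose, then chains four AnimDynamics nodes: linear springs for cans, angular springs for pots.
# Zipline enter/exit input events zero or restore three scale vectors, though the dynamics nodes show matching literal values rather than variable reads.
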AnimBlueprint /BRCosmetics/Characters/Player/Male/Medium/Bodies/M_MED_ChOneJonesy_Helmet/Meshes/Parts/M_MED_ChOneJonesy_Helmet_FaceAcc_AnimBP
target_skeleton=&/M_MED_ChOneJonesy_Helmet_FaceAcc_Skeleton
graphs [
 AnimGraph {nodes [
   Output Pose(Result=Use cached pose 'DynamicsCache'()) {layerGroup=None}
   Save cached pose 'inputPoseCache'(Pose=Copy Pose From Mesh() {bUseAttachedParent=true}) {cacheName=inputPoseCache}
   n1 = Anim Dynamics - Bone: dyn_can_lf!(
    ComponentPose=Local To Component(LocalPose=Use cached pose 'inputPoseCache'())
    Alpha=1
    ComponentLinearAccScale:Vector="0.4,0.4,0.4"
    ComponentLinearVelScale:Vector="0.4,0.4,0.4"
   ) -> (Pose) {
    angularDampingOverride=1
    angularSpringConstant=2000000
    bLinearSpring=true
    bUseGravityOverride=true
    gravityScale=0.35
    linearDampingOverride=1
    linearSpringConstant=50
    simulationSpace=BoneRelative
    boundBone {boneName=dyn_can_lf}
    componentLinearAccScale=<0.4 0.4 0.4>
    componentLinearVelScale=<0.4 0.4 0.4>
    gravityOverride {z=-980}
    physicsBodyDefinitions [{
      boundBone {boneName=dyn_can_lf}
      boxExtents=<20 20 20>
      constraintSetup {angularLimitsMax=<0 0 0>, angularLimitsMin=<0 0 0>, angularTarget=<0 0 0>, linearAxesMax=<0 0 0.6>}
      localJointOffset=<0 0 -8>
     }]
    relativeSpaceBone {boneName=base_can_lf}
    simSpaceSettings {maxAngularAcceleration=35, maxAngularVelocity=7, simSpaceAngularAlpha=1}
   }
   n2 = Anim Dynamics - Bone: dyn_can_rt!(ComponentPose=@n1, Alpha=1, ComponentLinearAccScale:Vector="0.35,0.35,0.35", ComponentLinearVelScale:Vector="0.35,0.35,0.35") -> (Pose) {
    angularDampingOverride=1
    angularSpringConstant=20000
    bLinearSpring=true
    bOverrideLinearDamping=true
    bUseGravityOverride=true
    gravityScale=0.35
    linearDampingOverride=1
    linearSpringConstant=50
    simulationSpace=BoneRelative
    boundBone {boneName=dyn_can_rt}
    componentLinearAccScale=<0.35 0.35 0.35>
    componentLinearVelScale=<0.35 0.35 0.35>
    gravityOverride {z=-980}
    physicsBodyDefinitions [{
      boundBone {boneName=dyn_can_rt}
      boxExtents=<20 20 20>
      constraintSetup {angularLimitsMax=<0 0 0>, angularLimitsMin=<0 0 0>, angularTarget=<1 0 0>, linearAxesMax=<0 0 0.45>}
      localJointOffset=<0 0 -8>
     }]
    relativeSpaceBone {boneName=base_can_rt}
    simSpaceSettings {maxAngularAcceleration=35, maxAngularVelocity=7, simSpaceAngularAlpha=1}
   }
   n3 = Anim Dynamics - Bone: dyn_pot_rt!(ComponentPose=@n2, Alpha=1, ComponentLinearAccScale:Vector="0.5,0.5,0.5", ComponentLinearVelScale:Vector="0.5,0.5,0.5") -> (Pose) {
    angularDampingOverride=1
    angularSpringConstant=20000
    bAngularSpring=true
    bOverrideAngularDamping=true
    bUseGravityOverride=true
    gravityScale=0.35
    simulationSpace=RootRelative
    boundBone {boneName=dyn_pot_rt}
    componentAppliedLinearAccClamp=<3000 3000 3000>
    componentLinearAccScale=<0.5 0.5 0.5>
    componentLinearVelScale=<0.5 0.5 0.5>
    gravityOverride {z=-980}
    physicsBodyDefinitions [{
      boundBone {boneName=dyn_pot_rt}
      boxExtents=<20 30 30>
      constraintSetup {angularLimitsMax=<0 10 0>, angularLimitsMin=<0 -20 0>, angularTarget=<1 0 0>}
      localJointOffset=<8 0 0>
     }]
    simSpaceSettings {maxAngularAcceleration=35, maxAngularVelocity=7, simSpaceAngularAlpha=1}
   }
   n4 = Anim Dynamics - Bone: dyn_pot_lf!(ComponentPose=@n3, Alpha=1, ComponentLinearAccScale:Vector="0.5,0.5,0.5", ComponentLinearVelScale:Vector="0.5,0.5,0.5") -> (Pose) {
    angularDampingOverride=1
    angularSpringConstant=20000
    bAngularSpring=true
    bGravityOverrideInSimSpace=true
    bOverrideAngularDamping=true
    bUseGravityOverride=true
    gravityScale=0.35
    simulationSpace=RootRelative
    boundBone {boneName=dyn_pot_lf}
    componentAppliedLinearAccClamp=<3000 3000 3000>
    componentLinearAccScale=<0.5 0.5 0.5>
    componentLinearVelScale=<0.5 0.5 0.5>
    gravityOverride {z=-980}
    physicsBodyDefinitions [{
      boundBone {boneName=dyn_pot_lf}
      boxExtents=<20 25 25>
      constraintSetup {angularLimitsMax=<0 10 0>, angularLimitsMin=<0 -15 0>, angularTarget=<1 0 0>}
      localJointOffset=<8 0 0>
     }]
    simSpaceSettings {maxAngularAcceleration=35, maxAngularVelocity=7, simSpaceAngularAlpha=1}
   }
   Save cached pose 'DynamicsCache'(Pose=Component To Local(ComponentPose=@n4)) {cacheName=DynamicsCache}
  ]}
 EventGraph {nodes [
   n1 = Event BlueprintProcessAnimInputEvent()
   n2 = IsNamed(self=@n1.AnimInputEvent, InEventName=aie.Player.Zipline.Entered)
   n3 = OR Boolean(A=Get bIsMoving2D(), B=Get bIsTacticalSprinting())
   n4 = Get bIsJumping()
   > n5 = Branch(Condition=OR Boolean(A=@n2, B=AND Boolean!(A=@n2, B=@n4, C:bool=@n3) -> (ReturnValue)))
   > Set animDyn on zipline pots(animDyn on zipline pots:Vector="0, 0, 0") -> (Output_Get:Vector)
   > Set animDyn on zipline can_r(animDyn on zipline can_r:Vector="0, 0, 0") -> (Output_Get:Vector)
   > Set animDyn on zipline can_l(animDyn on zipline can_l:Vector="0, 0, 0") -> (Output_Get:Vector)
   Branch(execute=@n5.else, Condition=IsNamed(self=@n1.AnimInputEvent, InEventName=aie.Player.Zipline.Exited))
   > Set animDyn on zipline pots(animDyn on zipline pots="0.5,0.5,0.5")
   > Set animDyn on zipline can_r(animDyn on zipline can_r="0.35,0.35,0.35")
   > Set animDyn on zipline can_l(animDyn on zipline can_l="0.4,0.4,0.4")
   Event BlueprintUpdateAnimation()
  ]}
]
variables [
 animDyn on zipline can_l {type=Vector, default=<0.4 0.4 0.4>}
 animDyn on zipline can_r {type=Vector, default=<0.35 0.35 0.35>}
 animDyn on zipline pots {type=Vector, default=<0.5 0.5 0.5>}
]
